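# Niagara system for the HomeRange player emote, named as a prop-spawn effect (variant 03), that plays a burst of dust and debris.
# Dust emits two burst groups, using spawn-group masks to vary size, color, forces and fade, and a visibility tag for its two sprite renderers.
# Curves drive dust SubUV frames and dynamic material parameters.
# A GPU emitter throws SubUV debris upward under drag and gravity; AirglowLWE has no renderer.
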NiagaraSystem /BRCosmetics/Animation/Game/MainPlayer/Emotes/HomeRange/VFX/NS_Emote_HomeRange_Propsspawn_FX_03
dependencies {
 usedDataInterfaces [&NiagaraDataInterfaceCurve]
 usedDynamicInputs [&Vector2DFromFloat, &FNTimeOfDayColor]
 usedModules [
  &SystemState
  &self:SystemSpawnScript.SystemScriptSource.SystemScriptGraph.NiagaraNodeAssignment_0.SetVariables_#1
 ]
}
emitters {
 AirglowLWE {rendererClasses [], renderers []}
 Dust {
  emitterUpdateScript [
   &EmitterState(Life Cycle Mode=::NewEnumerator1, Loop Behavior=::NewEnumerator1, Loop Duration=3, UseLoopDelay=false)
   &SpawnBurst_Instantaneous(Spawn Count=7, Use Spawn Probability=false, UseLoopCountLimit=false)
   &SpawnBurst_Instantaneous(Spawn Count=7, Spawn Group=1, Use Spawn Probability=false, UseLoopCountLimit=false) {module=SpawnBurst_Instantaneous001}
  ]
  particleSpawnScript [
   &InitializeParticle(
    Color=&MaskLinearColorBySpawnGroup(
     Masked Linear Color=&RandomRangeLinearColor(Minimum=<rgba 0.0908 0.0762 0.0685 0.6>, Maximum=<rgba 0.0931 0.063 0.0561 0.6>)
     Passthrough Linear Color=<rgba 0.2918 0.2159 0.1845 0.749>
    )
    Lifetime=&MaskFloatBySpawnGroup(
     Masked Float=&RandomRangeFloat(Minimum=1, Maximum=1.5)
     Passthrough Float=&RandomRangeFloat(Minimum=0.6, Maximum=1.3)
    )
    Sprite Rotation Mode=::NewEnumerator2
    Sprite Size Mode=::NewEnumerator3
    Uniform Sprite Size=&MaskFloatBySpawnGroup(
     Masked Float=&RandomRangeFloat(Minimum=50, Maximum=60)
     Passthrough Float=&RandomRangeFloat(Minimum=100, Maximum=110)
    )
    UsePositionOffset=false
    Write Material Random=true
   )
   /SetVariables(Particles.VisibilityTag=&SpawnGroup()) {isSetParameters=true}
   &AccelerationForce(Acceleration=&MaskVectorBySpawnGroup(Masked Vector=&UniformRangedVector(Minimum=<0 0 350>, Maximum=<0 0 800>), Passthrough Vector=<0 0 0>))
   &PointForce(
    Force Origin=&SimulationPosition()
    Force Strength=&MaskFloatBySpawnGroup(
     Masked Float=&RandomRangeFloat(Minimum=50, Maximum=70)
     Passthrough Float=&RandomRangeFloat(Minimum=100, Maximum=300)
    )
    Module.Use Falloff=false
    Module.UseLinearForce=false
    Particle Position:Vector3f=&AddVectorToPosition(
     Position=$Particles.Position
     Vector=&MaskVectorBySpawnGroup(Masked Vector=<0 0 0>, Passthrough Vector=<0 0 30>)
    )
   )
   &ShapeLocation(
    Hemisphere Distribution=<1 0.5>
    MaskSpawnGroup=false
    Offset=<0 0 -50>
    Offset Mode=::NewEnumerator0
    Override Seed=false
    Sphere Radius=&MaskFloatBySpawnGroup(Masked Float=5, Passthrough Float=0)
    Sphere Surface Distribution=0.381
    UseOwnerScale=true
   )
   &ApplyInitialForces()
   /SetVariables(Particles.Velocity=&Multiply_VectorByFloat(Vector=$Particles.Velocity, Float=/Niagara/DynamicInputs/Math/Min_Float(A:NiagaraFloat=0.5, B:NiagaraFloat=1.1))) {isSetParameters=true}
   &CameraOffset(Camera Offset Amount=&MaskFloatBySpawnGroup(Masked Float=&RandomRangeFloat(Minimum=20, Maximum=50), Passthrough Float=1))
   &AddVelocity(Velocity=&MaskVectorBySpawnGroup(Masked Vector=<0 0 0>, Passthrough Vector=&RandomRangeVector(Minimum=<0 0 100>, Maximum=<0 0 300>)))
  ]
  particleUpdateScript [
   &ParticleState()
   &SpriteRotationRate(Rotation Rate=&MaskFloatBySpawnGroup(
     Masked Float=&RandomRangeFloat(Minimum=-30, Maximum=30)
     Passthrough Float=&RandomRangeFloat(Minimum=-50, Maximum=50)
    ))
   &Drag(
    Drag=&MaskFloatBySpawnGroup(Masked Float=&RandomRangeFloat(Minimum=6, Maximum=8), Passthrough Float=1)
    Use Linear Drag=true
    Use Rotational Drag=true
   )
   &GravityForce(Gravity=&MaskVectorBySpawnGroup(Masked Vector=<0 0 0>, Passthrough Vector=<0 0 -980>))
   &SolveForcesAndVelocity(Clamp Velocity=false, Limit Acceleration=false)
   &ScaleSpriteSize(
    Scale Sprite Size Mode=::NewEnumerator0
    Uniform Scale Factor=&MaskFloatBySpawnGroup(
     Masked Float=&FloatFromCurve(FloatCurve=DataInterface() {propertyValues {Curve {keys [
          {interpMode=RCIM_Cubic, leaveTangent=2.8381, tangentMode=RCTM_Break, time=0, value=1.5}
          {interpMode=RCIM_Cubic, tangentMode=RCTM_Break, time=1, value=3.5}
         ]}}})
     Passthrough Float=&FloatFromCurve(FloatCurve=DataInterface() {propertyValues {Curve {keys [{arriveTangent=1.491, interpMode=RCIM_Cubic, leaveTangent=1.491, tangentMode=RCTM_User, time=0, value=1.2}, <at 1 3.49 RCIM_Cubic>]}}})
    )
   )
   /SetVariables(Particles.SubImageIndex=&FloatFromCurve(FloatCurve=DataInterface() {propertyValues {Curve {keys [
         {arriveTangent=29.8623, interpMode=RCIM_Cubic, leaveTangent=29.8623, tangentMode=RCTM_User, time=-0.0046, value=5.0556}
         <at 1.0011 22.1689 RCIM_Cubic>
        ]}}})) {isSetParameters=true}
   &DynamicMaterialParameters(
    Index 0 Param 1=7
    Index 0 Param 4=1
    Index 1 Param 1=&FloatFromCurve(FloatCurve=DataInterface() {propertyValues {Curve {keys [<at 0 0>, {arriveTangent=4.2257, interpMode=RCIM_Cubic, leaveTangent=4.2257, tangentMode=RCTM_User, time=0.25, value=0}]}}})
    Index 1 Param 2=&FloatFromCurve(FloatCurve=DataInterface() {propertyValues {Curve {keys [<at 0 0>]}}})
    Index 1 Param 3=&FloatFromCurve(FloatCurve=DataInterface() {propertyValues {Curve {keys [{interpMode=RCIM_Cubic, leaveTangent=6.3917, tangentMode=RCTM_Break, time=0, value=0}]}}})
    Index 1 Param 4=&FloatFromCurve(FloatCurve=DataInterface() {propertyValues {Curve {keys [{arriveTangent=-0.136, leaveTangent=-0.136, time=0, value=1}]}}})
    Index 2 Param 1=&FloatFromCurve(FloatCurve=DataInterface() {propertyValues {Curve {keys [
         {arriveTangent=2.646, interpMode=RCIM_Cubic, leaveTangent=2.646, tangentMode=RCTM_User, time=0, value=0}
         {leaveTangent=-0.0719, time=1, value=1}
        ]}}})
    Index 2 Param 2=&FloatFromCurve(
     FloatCurve=DataInterface() {propertyValues {Curve {keys [
         {arriveTangent=1.9056, time=0, value=0.2}
         {arriveTangent=0.0259, interpMode=RCIM_Cubic, leaveTangent=0.0259, tangentMode=RCTM_User, time=1, value=1.4}
        ]}}}
     Scale Curve=&RandomRangeFloat(Minimum=0.5, Maximum=1)
    )
    Index 2 Param 3=&FloatFromCurve(FloatCurve=DataInterface() {propertyValues {Curve {keys [
         {interpMode=RCIM_Cubic, leaveTangent=2.3064, tangentMode=RCTM_Break, time=0, value=1}
         {interpMode=RCIM_Cubic, leaveTangent=11.1233, tangentMode=RCTM_Break, time=1, value=2}
        ]}}})
    Index 2 Param 4=&FloatFromCurve(FloatCurve=DataInterface() {propertyValues {Curve {keys [
         {arriveTangent=-0.6867, interpMode=RCIM_Cubic, leaveTangent=-0.6867, tangentMode=RCTM_User, time=0, value=1}
         {interpMode=RCIM_Cubic, leaveTangent=-1.2956, tangentMode=RCTM_Break, time=1, value=0.75}
        ]}}})
    Module.Index0Float0Write=true
    Module.Index0Float1Write=true
    Module.Index0Float2Write=true
    Module.Index0Float3Write=true
    Module.Index1Float0Write=true
    Module.Index1Float1Write=true
    Module.Index1Float2Write=true
    Module.Index1Float3Write=true
    Module.Index2Float0Write=true
    Module.Index2Float1Write=true
    Module.Index2Float2Write=true
    Module.Index2Float3Write=true
    Write Parameter Index 1=::NewEnumerator1
    Write Parameter Index 2=::NewEnumerator1
   )
   &ScaleColor(
    Scale Alpha=&MaskFloatBySpawnGroup(
     Masked Float=&FloatFromCurve(FloatCurve=DataInterface() {propertyValues {Curve {keys [
          <at 0 0 RCIM_Cubic>
          {arriveTangent=0.3328, interpMode=RCIM_Cubic, leaveTangent=0.3328, tangentMode=RCTM_User, time=0.11, value=1}
          {arriveTangent=-0.194, interpMode=RCIM_Cubic, leaveTangent=-0.194, tangentMode=RCTM_User, time=0.5, value=1}
          <at 1 0>
         ]}}})
     Passthrough Float=&FloatFromCurve(FloatCurve=DataInterface() {propertyValues {Curve {keys [
          {arriveTangent=-0.4171, interpMode=RCIM_Cubic, leaveTangent=-0.4171, tangentMode=RCTM_User, time=0.2928, value=0.95}
          {arriveTangent=-4.2895, interpMode=RCIM_Cubic, leaveTangent=-4.2895, tangentMode=RCTM_User, time=1.0011, value=0.002}
         ]}}})
    )
    ScaleA=true
    ScaleRGB=true
   )
   /SetVariables(Particles.Rotation:NiagaraFloat=&RandomRangeFloat(Minimum=0, Maximum=360)) {isSetParameters=true}
  ]
  rendererClasses [&NiagaraSpriteRendererProperties]
  renderers [
   {rendererClass=&NiagaraSpriteRendererProperties}
   {rendererClass=&NiagaraSpriteRendererProperties, rendererIndex=1}
  ]
 }
 NE_FN_Impact_Debris_SubUV {
  simTarget=GPUComputeSim
  emitterUpdateScript [
   &EmitterState()
   &SpawnBurst_Instantaneous(Spawn Count=23, Use Spawn Probability=false, UseLoopCountLimit=false)
  ]
  particleSpawnScript [
   /SetVariables(Particles.RandomFloat=&UniformRangedFloat(Minimum=0, Maximum=3)) {isSetParameters=true}
   &InitializeParticle(
    Color=<rgba 0.4564 0.3712 0.3712 1>
    Lifetime Max=1.2
    Lifetime Min=0.9
    Lifetime Mode=::NewEnumerator1
    Sprite Rotation Mode=::NewEnumerator2
    Sprite Size Mode=::NewEnumerator1
    Uniform Sprite Size Max=15
    UsePositionOffset=false
    Write Material Random=true
   )
   &ShapeLocation(
    Handle Radius=50
    Large Radius=5
    MaskSpawnGroup=false
    Override Seed=false
    Shape Primitive=::NewEnumerator3
    UseOwnerScale=true
    V Distribution=0.5
   )
   &PointForce(
    Force Strength=&RandomRangeFloat(Minimum=150, Maximum=200)
    Module.Use Falloff=false
    Module.UseLinearForce=false
    Particle Position:Vector3f=&AddVectorToPosition(Position=$Particles.Position, Vector=<0 0 40>)
   )
   &ApplyInitialForces()
   /Niagara/Modules/Update/SubUV/SubUVAnimation(SubUV Animation Mode=::NewEnumerator1) {module=SubUVAnimation001}
   &AddVelocity(Velocity=&RandomRangeVector(Minimum=<0 0 240>, Maximum=<0 0 350>))
  ]
  particleUpdateScript [
   &ParticleState()
   &SpriteRotationRate(Rotation Rate=&RandomRangeFloat(Minimum=-20, Maximum=20)) {module=SpriteRotationRate001}
   &ScaleSpriteSize(Uniform Curve Sprite Scale=DataInterface() {propertyValues {Curve {keys [
        {interpMode=RCIM_Cubic, tangentMode=RCTM_User, time=0, value=0}
        {interpMode=RCIM_Cubic, tangentMode=RCTM_User, time=0.1, value=1}
        {interpMode=RCIM_Cubic, tangentMode=RCTM_User, time=0.66, value=1}
        {interpMode=RCIM_Cubic, tangentMode=RCTM_User, time=1, value=0}
       ]}}}) {module=ScaleSpriteSize001}
   &Drag(Drag=&RandomRangeFloat(Minimum=0.6, Maximum=0.7), Use Linear Drag=true, Use Rotational Drag=true)
   &GravityForce(Gravity=&RandomRangeVector(Minimum=<0 0 -1050>, Maximum=<0 0 -1200>))
   &SolveForcesAndVelocity(Clamp Velocity=false, Limit Acceleration=false)
  ]
  rendererClasses [&NiagaraSpriteRendererProperties]
  renderers [{rendererClass=&NiagaraSpriteRendererProperties}]
 }
}
summary {
 systemName=NS_Emote_HomeRange_Propsspawn_FX_03
 emitters [
  {emitterName=AirglowLWE, rendererClasses []}
  {emitterName=Dust, rendererClasses [&NiagaraSpriteRendererProperties]}
  {emitterName=NE_FN_Impact_Debris_SubUV, simTarget=GPUComputeSim, rendererClasses [&NiagaraSpriteRendererProperties]}
 ]
 userVariables []
}
