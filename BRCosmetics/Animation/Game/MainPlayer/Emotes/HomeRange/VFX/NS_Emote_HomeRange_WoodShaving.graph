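# Niagara system for the HomeRange player emote, named as a wood shaving effect.
# It holds a single emitter, Wood, with no renderers listed.
# The only module recorded is System State, with no user variables, data interfaces or dynamic inputs.
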NiagaraSystem /BRCosmetics/Animation/Game/MainPlayer/Emotes/HomeRange/VFX/NS_Emote_HomeRange_WoodShaving
dependencies {usedDataInterfaces [], usedDynamicInputs [], usedModules [&SystemState]}
emitters {Wood {rendererClasses [], renderers []}}
summary {systemName=NS_Emote_HomeRange_WoodShaving, emitters [{emitterName=Wood, rendererClasses []}], userVariables []}
